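# Translucent unlit glow material for Niagara mesh particles, named for the Jam playspace border.
# A color-curve texture, sampled along a slowly panned U coordinate, is scaled by time-of-day parameters and particle color, then fed through MF_FX_EmissiveAdjustments for emissive.
# Particle alpha times a parameterized SmoothStep over the V coordinate forms a gradient, apparently for opacity.
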Material /FMJamSystem/Effects/Playspace/Materials/M_JamPlayspace_BorderGlow
graph {
 properties {
  bUsedWithNiagaraMeshParticles=true
  blendMode=BLEND_Translucent
  pixelDepthOffsetMode=PDOM_Legacy
  shadingModel=MSM_Unlit
  usedShadingModels=Unlit
 }
 nodes [
  n1 = MaterialFunctionCall(
   Value_Morning=ScalarParameter() {defaultValue=1, parameterName=Morning}
   Value_Day=ScalarParameter() {defaultValue=1, parameterName=Day}
   Value_Evening=ScalarParameter() {defaultValue=1, parameterName=Evening}
   Value_Night=ScalarParameter() {defaultValue=1, parameterName=Night}
  ) {materialFunction=&Scalar_Time_of_Day}
  n2 = TextureCoordinate()
  n3 = TextureSample(UVs=ComponentMask(None=Panner(Coordinate=@n2) {speedX=-0.05}) {r=true}) {texture=/FMJamSystem/Effects/Playspace/Textures/T_JamPlayspace_ColorCurve}
  n4 = ParticleColor()
  MaterialFunctionCall(Emissive=Multiply(A=Multiply(A=@n1.Result, B=@n3.RGB), B=@n4.RGB)) {materialFunction=&MF_FX_EmissiveAdjustments}
  n5 = TextureCoordinate()
  Multiply(
   A=@n4.A
   B=SmoothStep(Min=ScalarParameter(), Max=ScalarParameter() {defaultValue=6, parameterName=Param_1}, Value=ComponentMask(None=@n5) {g=true})
  )
 ]
}
