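# Translucent, two-sided unlit material for Niagara mesh particles that draws the inner border of a Jam playspace.
# Emissive combines a panning color-curve texture with a boosted V gradient, tinted by particle color.
# Opacity layers that gradient with sine-warped procedural stripes, scaled by particle alpha.
# A linear-or-mirrored parameter reshapes gradient and stripes and blends in depth fade, while LockedJam? multiplies opacity by animated, rotated bands.
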Material /FMJamSystem/Effects/Playspace/Materials/M_JamPlayspace_Border_Inner
graph {
 properties {
  bUsedWithNiagaraMeshParticles=true
  blendMode=BLEND_Translucent
  pixelDepthOffsetMode=PDOM_Legacy
  shadingModel=MSM_Unlit
  twoSided=true
  usedShadingModels=Unlit
 }
 nodes [
  n1 = ParticleColor()
  n2 = ComponentMask(None=TextureCoordinate()) {g=true}
  n3 = Multiply(A=@n2) {constB=0.5}
  n4 = LinearInterpolate(
   A=Power(Base=@n2) {constExponent=3}
   B=Power(Base=Sine(None=@n3)) {constExponent=3}
   Alpha=ScalarParameter() {parameterName=Linear or Mirrored Gradient}
  )
  n5 = NamedRerouteUsage()
  n6 = Multiply(A=@n1.A, B=Add(A=Multiply(A=@n4) {constB=0.07, desc=Gradient Opacity}, B=Multiply(A=@n5) {constB=0.4, desc=Stripes Opacity}))
  n7 = LinearInterpolate(
   A=@n6
   B=DepthFade(Opacity=@n6) {fadeDistanceDefault=200, opacityDefault=0}
   Alpha=ScalarParameter() {parameterName=Linear or Mirrored Gradient}
  )
  LinearInterpolate(A=@n7, B=Multiply(A=@n7, B=NamedRerouteUsage()), Alpha=ScalarParameter() {defaultValue=1, parameterName=LockedJam?}) {constAlpha=0}
  n8 = ComponentMask(None=TextureCoordinate()) {g=true}
  n9 = Sine(None=Multiply(A=@n8) {constB=0.5})
  n10 = ScalarParameter() {parameterName=Linear or Mirrored Gradient}
  n11 = TextureCoordinate()
  n12 = TextureSample(UVs=ComponentMask(None=Panner(Coordinate=@n11) {speedX=-0.05}) {r=true}) {texture=/FMJamSystem/Effects/Playspace/Textures/T_JamPlayspace_ColorCurve}
  n13 = Multiply(A=Power(Base=LinearInterpolate(A=@n8, B=@n9, Alpha=@n10)) {constExponent=10}, B=Multiply(B=@n12.RGB) {constA=10})
  n14 = Multiply(
   A=@n12.RGB
   B=MaterialFunctionCall(Value_Morning=Constant() {r=1}, Value_Day=Constant() {r=1}, Value_Evening=Constant() {r=1}, Value_Night=Constant() {r=1}) {materialFunction=&Scalar_Time_of_Day}.Result
  )
  MaterialFunctionCall(Emissive=Multiply(A=Add(A=@n13, B=@n14), B=@n1.RGB)) {materialFunction=&MF_FX_EmissiveAdjustments}
  n15 = ScalarParameter() {defaultValue=0.2, parameterName=Param_3}
  n16 = ScalarParameter() {defaultValue=0.9, parameterName=Param_4}
  n17 = Frac(None=Multiply(A=Time()) {constB=-0.5, desc=PanSpeed})
  n18 = ScalarParameter() {defaultValue=40, parameterName=Param_1}
  n19 = ScalarParameter() {defaultValue=1, parameterName=Param_2}
  n20 = MaterialFunctionCall(UVs=Multiply(A=TextureCoordinate(), B=AppendVector(A=@n18, B=@n19)), "Rotation Angle (0-1)"=ScalarParameter() {defaultValue=0.4}) {materialFunction=&CustomRotator}
  NamedRerouteDeclaration(None=LinearInterpolate(
    A=ScalarParameter() {defaultValue=0.25, parameterName=Locked Alpha Min}
    Alpha=MaterialFunctionCall!(Width=@n15, Sharpness=@n16, Offset=@n17, Direction Switch, Compare, Input Coordinates=@n20.Rotated Values) {materialFunction=/Engine/Functions/Engine_MaterialFunctions03/Procedurals/GeneratedBand}.Result
   ) {constA=0.2}) {name=LockedJam, nodeColor=<rgba 0.6875 1 0 1>}
  n21 = Time()
  n22 = TextureCoordinate()
  n23 = Multiply(A=@n22) {constB=10}
  n24 = Multiply(A=@n22) {constB=14}
  n25 = ScalarParameter() {parameterName=Linear or Mirrored Gradient}
  n26 = Add(A=ComponentMask(None=TextureCoordinate() {uTiling=20}) {r=true}, B=Multiply(A=Time()) {constB=0.05, desc=Line Warp Speed})
  n27 = Sine(None=Multiply(A=Time()) {constB=0.25})
  n28 = Constant() {r=-1}
  n29 = Constant() {r=1}
  n30 = Multiply(
   A=Sine(None=Multiply(A=@n26))
   B=LinearInterpolate(Alpha=MaterialFunctionCall(Input=@n27, Input Low=@n28, Input High=@n29) {materialFunction=&RemapValue_0-1}.Result) {constA=-0.6, constB=0.6}
  ) {constB=0.1}
  n31 = OneMinus(None=ComponentMask(None=TextureCoordinate() {vTiling=2}) {g=true})
  n32 = MaterialFunctionCall(
   Width=Constant() {r=0.2}
   Offset=Frac(None=Multiply(A=@n21) {constB=0, desc=PanSpeed})
   Direction Switch=StaticBool()
   Input Coordinates=Add(A=LinearInterpolate(A=@n23, B=@n24, Alpha=@n25), B=Multiply(A=@n30, B=@n31))
  ) {materialFunction=&GeneratedBand}
  n33 = ComponentMask(None=TextureCoordinate()) {g=true}
  n34 = Sine(None=Multiply(A=@n33) {constB=0.5})
  n35 = Power(Base=LinearInterpolate(
    A=SmoothStep(Value=@n33) {constMin=0.1}
    B=Power(Base=@n34) {constExponent=3}
    Alpha=ScalarParameter() {parameterName=Linear or Mirrored Gradient}
   )) {constExponent=0.8}
  NamedRerouteDeclaration(None=Multiply(A=Multiply(A=@n32.Result, B=@n35)) {constB=0.1}) {name=Stripes, nodeColor=<rgba 1 0 0.2109 1>}
 ]
}
