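# Translucent unlit material for the outer border of the Jam playspace effect, built for Niagara mesh particles with SplineThicken driven by a LineWidth parameter.
# A vertically panning color-curve texture is tinted by particle color, scaled by time of day, then passed through FX emissive adjustments.
# Three summed panning noise samples, particle alpha and a sharpened depth fade form the opacity mask.
# A banded "Stripes" reroute is declared but never used.
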Material /FMJamSystem/Effects/Playspace/Materials/M_JamPlayspace_Border_Outer
graph {
 properties {
  bUsedWithNiagaraMeshParticles=true
  blendMode=BLEND_Translucent
  pixelDepthOffsetMode=PDOM_Legacy
  shadingModel=MSM_Unlit
  twoSided=true
  usedShadingModels=Unlit
 }
 nodes [
  n1 = MaterialFunctionCall(
   Value_Morning=ScalarParameter() {defaultValue=1, parameterName=Morning}
   Value_Day=ScalarParameter() {defaultValue=1, parameterName=Day}
   Value_Evening=ScalarParameter() {defaultValue=1, parameterName=Evening}
   Value_Night=ScalarParameter() {defaultValue=1, parameterName=Night}
  ) {materialFunction=&Scalar_Time_of_Day}
  n2 = TextureCoordinate()
  n3 = TextureSample(UVs=ComponentMask(None=Panner(Coordinate=@n2) {speedY=0.25}) {g=true}) {texture=/FMJamSystem/Effects/Playspace/Textures/T_JamPlayspace_ColorCurve}
  n4 = ParticleColor()
  MaterialFunctionCall(Emissive=Multiply(A=Multiply(A=@n1.Result, B=@n3.RGB), B=@n4.RGB)) {materialFunction=&MF_FX_EmissiveAdjustments}
  n5 = ScalarParameter() {defaultValue=4, parameterName=LineWidth}
  MaterialFunctionCall(WidthBase=@n5, WidthTip=@n5) {materialFunction=&SplineThicken}
  n6 = ComponentMask(None=Panner(Coordinate=TextureCoordinate() {vTiling=4}) {speedY=0.1}) {g=true}
  n7 = ComponentMask(None=Panner(Coordinate=TextureCoordinate() {vTiling=5}) {speedY=0.1}) {g=true}
  n8 = Panner(Coordinate=TextureCoordinate() {vTiling=3}) {speedY=0.2}
  n9 = Add(
   A=Add(
    A=TextureSample(UVs=@n6) {samplerType=SAMPLERTYPE_Grayscale, texture=/Game/TimeOfDay/TODM_Sequence/Materials/Sky/Textures/LowResBlurredNoise_Radish_LF_Root}.R
    B=TextureSample(UVs=@n7) {samplerType=SAMPLERTYPE_Grayscale, texture=&LowResBlurredNoise_Radish_LF_Root}.R
   )
   B=TextureSample(UVs=ComponentMask(None=@n8) {g=true}) {samplerType=SAMPLERTYPE_Grayscale, texture=&LowResBlurredNoise_Radish_LF_Root}.R
  )
  n10 = DepthFade() {fadeDistanceDefault=30}
  Multiply(A=Multiply(A=@n4.A, B=Multiply(A=@n9) {constB=0.1}), B=Saturate(None=Power(Base=@n10) {constExponent=3}))
  n11 = Time()
  n12 = MaterialFunctionCall!(
   Width=Constant() {r=0.5}
   Sharpness
   Offset=Frac(None=Multiply(A=@n11) {constB=0, desc=PanSpeed})
   Direction Switch=StaticBool()
   Compare
   Input Coordinates=Multiply(A=TextureCoordinate() {vTiling=10}) {constB=0.5}
  ) {materialFunction=/Engine/Functions/Engine_MaterialFunctions03/Procedurals/GeneratedBand}
  n13 = TextureCoordinate()
  n14 = Power(Base=SmoothStep(Value=ComponentMask(None=@n13) {g=true}) {constMin=0.1}) {constExponent=0.8}
  NamedRerouteDeclaration(None=Multiply(A=Multiply(A=@n12.Result, B=@n14))) {name=Stripes, nodeColor=<rgba 1 0 0.2109 1>}
 ]
}
