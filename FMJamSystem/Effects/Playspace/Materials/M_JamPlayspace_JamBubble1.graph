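# Translucent, unlit, two-sided bubble material for the Jam playspace effect, usable on Niagara mesh particles.
# A mask sums an inverted depth-fade intersection glow, a Fresnel rim scaled by a vertical UV gradient, and GeneratedBand stripes confined to a soft sine-shaped line.
# Color comes from a diagonally panning color-curve texture, scaled by time of day and passed through MF_FX_EmissiveAdjustments.
# The stripe pan speed is zero, so the stripes stay static.
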Material /FMJamSystem/Effects/Playspace/Materials/M_JamPlayspace_JamBubble1
graph {
 properties {
  bUsedWithNiagaraMeshParticles=true
  blendMode=BLEND_Translucent
  pixelDepthOffsetMode=PDOM_Legacy
  shadingModel=MSM_Unlit
  twoSided=true
  usedShadingModels=Unlit
 }
 nodes [
  n1 = DepthFade(FadeDistance=ScalarParameter() {defaultValue=20, parameterName=DepthFade})
  n2 = Divide(
   A=Subtract(A=OneMinus(None=@n1), B=ScalarParameter() {defaultValue=0.1, parameterName="DepthFade Size (Sub)"})
   B=ScalarParameter() {defaultValue=0.5, parameterName="DepthFade Softness (Div)"}
  )
  n3 = Fresnel(
   ExponentIn=ScalarParameter() {defaultValue=8, parameterName=Fresnel Exponent}
   BaseReflectFractionIn=ScalarParameter() {defaultValue=0.04, parameterName=Fresnel BaseReflectFrac}
  )
  n4 = Power(Base=ComponentMask(None=TextureCoordinate()) {g=true}) {constExponent=0.5}
  n5 = ScalarParameter() {defaultValue=0.2, parameterName=Lines Width}
  n6 = Frac(None=Multiply(A=Time()) {constB=0, desc=PanSpeed})
  n7 = StaticBool()
  n8 = Multiply(A=TextureCoordinate(), B=ScalarParameter() {defaultValue=200, parameterName=Stripe Count}) {constB=14}
  n9 = TextureCoordinate()
  n10 = Sine(None=Multiply(A=ComponentMask(None=@n9) {g=true}) {constB=0.5})
  n11 = ScalarParameter() {defaultValue=0.95, parameterName="Line Size (Sub)"}
  n12 = Saturate(None=Divide(A=Subtract(A=@n10, B=@n11), B=ScalarParameter() {defaultValue=0.1, parameterName="Line Gradietn Softness (Div)"}))
  Add(
   A=Add(A=Saturate(None=@n2), B=Multiply(A=@n3, B=@n4))
   B=Multiply(
    A=MaterialFunctionCall!(Width=@n5, Sharpness, Offset=@n6, Direction Switch=@n7, Compare, Input Coordinates=@n8) {materialFunction=/Engine/Functions/Engine_MaterialFunctions03/Procedurals/GeneratedBand}.Result
    B=Power(Base=@n12) {constExponent=0.8}
   )
  )
  n13 = ScalarParameter() {defaultValue=1, parameterName=Morning}
  n14 = ScalarParameter() {defaultValue=1, parameterName=Day}
  n15 = ScalarParameter() {defaultValue=1, parameterName=Evening}
  n16 = ScalarParameter() {defaultValue=1, parameterName=Night}
  n17 = Panner() {speedX=0.5, speedY=0.5}
  MaterialFunctionCall(Emissive=Multiply(
    A=MaterialFunctionCall(Value_Morning=@n13, Value_Day=@n14, Value_Evening=@n15, Value_Night=@n16) {materialFunction=&Scalar_Time_of_Day}.Result
    B=TextureSample(UVs=@n17) {texture=/FMJamSystem/Effects/Playspace/Textures/T_JamPlayspace_ColorCurve}.RGB
   )) {materialFunction=&MF_FX_EmissiveAdjustments}
 ]
}
